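# Masked, unlit Niagara sprite material for the Missus Mind goop character's goo particles, faking lit liquid.
# Size-scaled, time-crossfaded noise UVs feed an erosion mask, shaped by UV edge falloff and a particle erosion parameter.
# Noise, a panning normal map and derived edge normals build a reflection vector for sky-light-driven diffuse, highlight and rim terms.
# Dynamic parameters control night diffuse, time-of-day and light intensities, and a cubemap reflection blends in by scalar.
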Material /BRCosmetics/Characters/Player/Female/Medium/Bodies/F_MED_MissusMind_Goop/FX/M_Character_MissusMind_Goo_Sprite_02
graph {
 properties {
  bAutomaticallySetUsageInEditor=false
  bUsedWithNiagaraSprites=true
  blendMode=BLEND_Masked
  previewMesh=/Engine/Transient.MaterialEditorMeshComponent_78
  shadingModel=MSM_Unlit
  translucencyPass=MTP_BeforeDOF
  usedShadingModels=Unlit
 }
 nodes [
  Constant3Vector() {constant=<rgba 0.0151 0.0272 0.0525 1>}
  Constant() {r=0.8}
  Constant() {r=0.25}
  Constant() {r=0.5}
  n1 = TextureCoordinate()
  n2 = ParticleSize()
  n3 = ScalarParameter() {defaultValue=0.066, parameterName=UV Scale}
  n4 = VectorParameter() {parameterName=UV Scales, defaultValue=<rgba 1 0.75 0.6 0>}
  n5 = MaterialFunctionCall!(
   UV=Multiply(A=Subtract(A=@n1) {constB=0.5}, B=Multiply(A=@n2, B=@n3))
   UV Scale Pivot=Constant()
   UV Scale Start=@n4.R
   UV Scale End=@n4.R
   Duration=@n4.R
   Time=Add(A=Time(), B=DynamicParameter() {paramNames [Edges Mask Exponent U, Edges Mask Exponent V, Erosion, Random Time Offset]}.Random Time Offset)
   Random UV Offset Enable=StaticBool() {value=true}
   Random UV Offset Scale
  ) {materialFunction=/BRCosmetics/Accessories/FORT_Backpacks/Backpack_M_MED_CrystalGlobe/FX/Materials/MF_ScaleCrossfadeUV}
  n6 = TextureObject() {
   samplerType=SAMPLERTYPE_LinearColor
   texture=/BRCosmetics/Accessories/FORT_Backpacks/Backpack_M_MED_CrystalGlobe/FX/Textures/T_Backpack_CrystalGlobe_Noise
  }
  n7 = TextureSample(UVs=@n5.UV 1, Tex=@n6) {
   samplerType=SAMPLERTYPE_LinearColor
   texture=&T_Backpack_CrystalGlobe_Noise
  }
  n8 = TextureSample(UVs=@n5.UV 2, Tex=@n6) {
   samplerType=SAMPLERTYPE_LinearColor
   texture=&T_Backpack_CrystalGlobe_Noise
  }
  n9 = LinearInterpolate(A=@n7.RGBA, B=@n8.RGBA, Alpha=@n5.Crossfade Alpha)
  n10 = ScalarParameter() {defaultValue=1, parameterName=NormalUVScale}
  n11 = TextureSample(UVs=Panner(Coordinate=Multiply(A=@n5.UV 1, B=@n10)) {speedY=0.2}) {samplerType=SAMPLERTYPE_Normal, texture=/Game/Characters/Player/Male/Medium/Bodies/M_MED_Mastermind/Textures/T_M_MED_Mastermind_Twisted_N}
  n12 = ConstantBiasScale(None=ComponentMask(None=MaterialFunctionCall(BaseNormal=@n9, AdditionalNormal=@n11.RGBA) {materialFunction=&BlendAngleCorrectedNormals}.Result) {b=true, g=true, r=true}) {bias=-0.5, scale=2}
  n13 = ScalarParameter() {defaultValue=0.1, parameterName=Normal Flatness}
  n14 = Normalize(VectorInput=Add(
    A=MaterialFunctionCall(Normal=@n12, Flatness=@n13) {materialFunction=&FlattenNormal}.Result
    B=NamedRerouteUsage()
   ))
  n15 = ReflectionVectorWS(CustomWorldNormal=ComponentMask(None=Transform(None=@n14)) {b=true, g=true, r=true})
  n16 = DotProduct(A=SkyAtmosphereLightDirection(), B=@n15)
  n17 = DynamicParameter() {parameterIndex=3, paramNames [HighlightR,HighlightG,HighlightB,Param4]}
  n18 = Multiply(
   A=Power(Base=ConstantBiasScale(None=@n16)) {constExponent=24}
   B=Multiply(A=MaterialFunctionCall(R=@n17.HighlightR, G=@n17.HighlightR, B=@n17.HighlightR) {materialFunction=&AppendMany}.RGB) {constB=0.3}
  ) {constB=0}
  n19 = Multiply(A=Constant() {r=0.05}, B=Multiply(A=Saturate(None=@n16), B=SkyAtmosphereDistantLightScatteredLuminance()))
  n20 = ConstantBiasScale(None=TextureCoordinate()) {bias=-0.5, scale=2}
  n21 = DynamicParameter() {paramNames [Edges Mask Exponent U, Edges Mask Exponent V, Erosion, Random Time Offset]}
  n22 = Power(Base=Abs(None=@n20), Exp=AppendVector(A=@n21.Edges Mask Exponent U, B=@n21.Edges Mask Exponent U)) {constExponent=3}
  n23 = MaterialFunctionCall(Vector 2=@n22) {materialFunction=/Engine/Functions/Engine_MaterialFunctions02/Utility/VectorLength}
  n24 = MaterialFunctionCall!(
   Gradient=ComponentMask(None=@n9) {a=true}
   Erosion Amount=Max(
    A=Saturate(None=@n23.V2 Length)
    B=DynamicParameter() {paramNames [Edges Mask Exponent U, Edges Mask Exponent V, Erosion, Random Time Offset]}.Erosion
   )
   Softness=ScalarParameter() {defaultValue=0.05, parameterName=Erosion Softness}
  ) {materialFunction=/BRCosmetics/Athena/Effects/Particles/Skydiving/NS_Contrail_DarkAndLight/Functions/MF_Erosion}
  n25 = DynamicParameter() {parameterIndex=1, paramNames [DiffuseNight,TODMulti,Param3,Param4]}
  n26 = VectorParameter() {parameterName=Diffuse, defaultValue=<rgba 0.015 0.015 0.015 0>}
  n27 = Constant3Vector() {constant=<rgba 0.0052 0.0068 0.015 1>}
  n28 = Multiply(A=@n24.Increased Range, B=Multiply(A=@n25.TODMulti, B=LinearInterpolate(A=@n26.RGB, B=@n27, Alpha=@n25.DiffuseNight)))
  n29 = OneMinus(None=SkyAtmosphereLightDirection())
  n30 = Multiply(A=Power(Base=DotProduct(A=@n29, B=@n15)) {constExponent=4}) {constB=0.35}
  n31 = DynamicParameter() {parameterIndex=2, paramNames [RimLightR,RimLightG,RimLightB,Param4]}
  n32 = MaterialFunctionCall(R=@n31.RimLightR, G=@n31.RimLightR, B=@n31.RimLightR) {materialFunction=&AppendMany}
  n33 = OneMinus(None=Length(None=@n24.Default Range) {is=class MaterialExpressionLength})
  n34 = Constant3Vector() {constant=<rgba 0.025 0.025 0.025 0>}
  n35 = Add(
   A=Add(A=Add(A=@n19, B=@n28), B=Multiply(A=@n30, B=@n32.RGB))
   B=Multiply(A=Power(Base=@n33) {constExponent=4}, B=Multiply(A=@n25.TODMulti, B=@n34))
  )
  n36 = LinearInterpolate(A=Multiply(A=Add(A=@n18, B=@n35), B=ParticleColor().RGB), Alpha=ScalarParameter() {parameterName=NoEmissiveLerp}) {constB=0}
  n37 = MaterialFunctionCall(Normal=NamedRerouteUsage(), Flatness=ScalarParameter() {defaultValue=0.5, parameterName=Reflection_NormalFlatten}) {materialFunction=&FlattenNormal}
  n38 = TextureSampleParameterCube!(UVs=ReflectionVectorWS(CustomWorldNormal=Transform(None=@n37.Result)), Apply View MipBias) {parameterName=ReflectHDRI, samplerType=SAMPLERTYPE_LinearColor, texture=/Game/Effects/FlipSimulation/Rendering/T_WaterReflection}
  n39 = Multiply(A=@n38.G) {constB=0.02}
  MaterialFunctionCall(Emissive=LinearInterpolate(A=@n36, B=Add(A=@n36, B=@n39), Alpha=ScalarParameter() {parameterName=UseEmissiveHighReflectionHDRI})) {materialFunction=&MF_FX_EmissiveAdjustments}
  Multiply() {constB=0}
  NamedRerouteDeclaration(None=@n24.Increased Range) {name=To Opacity, nodeColor=<rgba 0.8047 1 0 1>}
  n40 = Multiply(A=@n22, B=Sign(None=@n20))
  n41 = ScalarParameter() {defaultValue=2, parameterName=Edge Normal Intensity}
  NamedRerouteDeclaration(None=DeriveNormalZ(InXY=Multiply(A=@n40, B=@n41))) {name=Edge Normal, nodeColor=<rgba 0.0312 1 0 1>}
  NamedRerouteUsage()
  VectorParameter() {parameterName=Specular Highlight, defaultValue=<rgba 0.4397 0.2582 0.1559 1>}
 ]
}
