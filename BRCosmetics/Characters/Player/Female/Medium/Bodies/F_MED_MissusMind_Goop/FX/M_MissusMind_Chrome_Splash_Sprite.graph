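# Translucent lit Niagara sprite material for a chrome splash effect on the MissusMind Goop character.
# Color blends from particle color to that color multiplied by a splash texture, driven by the TextureColorLerp dynamic parameter.
# An erosion mask subtracts particle alpha from the texture's red channel, then sharpens it with power, multiply and saturate.
# A normal map, constant surface values and generated spherical particle normals support the lit liquid look.
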Material /BRCosmetics/Characters/Player/Female/Medium/Bodies/F_MED_MissusMind_Goop/FX/M_MissusMind_Chrome_Splash_Sprite
graph {
 properties {
  bAutomaticallySetUsageInEditor=false
  bCastRayTracedShadows=false
  bGenerateSphericalParticleNormals=true
  bUsedWithNiagaraSprites=true
  blendMode=BLEND_Translucent
  translucencyLightingMode=TLM_Surface
  translucencyPass=MTP_BeforeDOF
  usedShadingModels=Default Lit
 }
 nodes [
  Constant() {r=0.8}
  Constant() {r=0.25}
  Constant() {r=0.5}
  n1 = ParticleColor()
  n2 = TextureSampleParameter2D() {parameterName=BaseTexture, texture=/Game/Effects/Fort_Effects/Textures/Liquid/T_FX_Splash_01}
  LinearInterpolate(
   A=@n1.RGB
   B=Multiply(A=@n1.RGB, B=@n2.RGB) {constB=0.2}
   Alpha=DynamicParameter() {paramNames [TextureColorLerp, Param2, Param3, Param4]}.TextureColorLerp
  )
  n3 = Subtract(A=@n2.R, B=@n1.A) {constB=0.396}
  Saturate(None=Multiply(A=Power(Base=@n3) {constExponent=0.5}) {constB=2})
  TextureSampleParameter2D() {parameterName=NormalTexture, samplerType=SAMPLERTYPE_Normal, texture=/Game/Effects/Fort_Effects/Textures/Liquid/T_FX_Splash_01_N}
 ]
}
