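# Niagara system for the Missus Mind Goop character cosmetic, emitting idle goop sprites from the body in gameplay and front end.
# One sprite emitter splits randomized spawns into eight groups, each pinned to a skeletal mesh bone or socket with an offset.
# Age curves push, align and stretch sprites along bone rotation, while user bools zero selected groups for hidden hands or feet.
# Time-of-day curves drive material, rim and highlight values, with front-end color overrides.
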NiagaraSystem /BRCosmetics/Characters/Player/Female/Medium/Bodies/F_MED_MissusMind_Goop/FX/NS_Character_MissusMind_Goop_Idle_Body
dependencies {
 usedDataInterfaces [
  &NiagaraDataInterfaceCurve
  &NiagaraDataInterfaceArrayFloat
  &NiagaraDataInterfaceVector2DCurve
  &NiagaraDataInterfaceArrayInt32
  &NiagaraDataInterfaceArrayFloat3
  &NiagaraDataInterfaceVectorCurve
  &NiagaraDataInterfaceColorCurve
  &NiagaraDataInterfaceSkeletalMesh
 ]
 usedDynamicInputs [
  &MaskVector2dBySpawnGroup
  &Vector2DFromFloat
  &RandomRangeVector2D
  &MakeFloatFromVector2d
 ]
 usedModules [
  &self:SystemSpawnScript.SystemScriptSource.SystemScriptGraph.NiagaraNodeAssignment_0.SetVariables_#1
  &SystemState
 ]
}
emitters {Goop {
  emitterUpdateScript [
   &EmitterState()
   &SpawnRate(
    Spawn Probability=0.33
    SpawnRate=&RandomRangeFloat(Minimum=30, Maximum=17)
    Use Spawn Probability=true
   ) {module=SpawnRate002}
  ]
  particleSpawnScript [
   &FN_SetSpawnGroup(Number of Spawn Groups=8)
   &InitializeParticle(
    Color=&LerpLinearColors(
     StartColor=<rgba 1 1 1 9>
     EndColor=<rgba 0.28 0.28 0.28 9>
     LerpFactor=&FN_FrontEndFloatModifier(Gameplay=0, Front End=1)
    )
    Lifetime=&RandomRangeFloat(Minimum=0.5, Maximum=1)
    Sprite Size Mode=::NewEnumerator1
    Uniform Sprite Size Max=5.2
    Uniform Sprite Size Min=3
    UsePositionOffset=false
    Write Material Random=true
   )
   &CameraOffset(Camera Offset Amount=2)
  ]
  particleUpdateScript [
   &ParticleState()
   &ScaleSpriteSize(
    Scale Sprite Size Mode=::NewEnumerator0
    Uniform Scale Factor=&MakeCustomFloatFromBool(
     Bool=$User.bisHandHidden
     True Float=&SelectFloatFromArray(
      Direct Array Index=&SpawnGroup()
      Float Selection Array=DataInterface() {propertyValues {FloatData [0,1,0,1,0,1,0,1]}}
     )
     False Float=1
    )
   ) {module=ScaleSpriteSize002}
   &ScaleSpriteSize(
    Scale Sprite Size Mode=::NewEnumerator0
    Uniform Scale Factor=&MakeCustomFloatFromBool(
     Bool=$User.bisFootHidden
     True Float=&SelectFloatFromArray(
      Direct Array Index=&SpawnGroup()
      Float Selection Array=DataInterface() {propertyValues {FloatData [1,1,1,1,1,0,1,0,1]}}
     )
     False Float=1
    )
   ) {module=ScaleSpriteSize003}
   &ScaleSpriteSize(
    Uniform Curve Scale=1.2
    Uniform Curve Sprite Scale=DataInterface() {propertyValues {Curve {keys [
        <at 0 0 RCIM_Cubic>
        <at 0.2674 1>
        {arriveTangent=-1.3651, interpMode=RCIM_Cubic, leaveTangent=-1.3651, time=0.7, value=1}
        {tangentMode=RCTM_User, time=1, value=0}
       ]}}}
   ) {module=ScaleSpriteSize001}
   &ScaleSpriteSize(
    Scale Factor=&Lerp_Vector2(
     A=&Vector2DFromCurve(
      Vector2Curve=DataInterface() {propertyValues {
        XCurve {keys [{tangentMode=RCTM_User, time=0, value=1}, {tangentMode=RCTM_User, time=1, value=1}]}
        YCurve {keys [
          {interpMode=RCIM_Cubic, leaveTangentWeight=0.3333, tangentMode=RCTM_User, time=0, value=1.3}
          {arriveTangentWeight=0.1313, interpMode=RCIM_Cubic, leaveTangentWeight=0.2021, tangentMode=RCTM_User, time=0.3938, value=2.5}
          {arriveTangentWeight=0.1354, interpMode=RCIM_Cubic, leaveTangentWeight=0.0667, tangentMode=RCTM_User, time=0.8, value=1}
          {tangentMode=RCTM_User, time=1, value=1}
         ]}
       }}
      ScaleCurve=<1 1>
     )
     B=&Vector2DFromCurve(
      Vector2Curve=DataInterface() {propertyValues {
        XCurve {keys [{tangentMode=RCTM_User, time=0, value=1}, {tangentMode=RCTM_User, time=1, value=1}]}
        YCurve {keys [
          {interpMode=RCIM_Cubic, leaveTangentWeight=0.3333, tangentMode=RCTM_User, time=0, value=1}
          {arriveTangentWeight=0.1313, interpMode=RCIM_Cubic, leaveTangentWeight=0.2021, tangentMode=RCTM_User, time=0.3938, value=1.3}
          {arriveTangentWeight=0.1354, interpMode=RCIM_Cubic, leaveTangentWeight=0.0667, tangentMode=RCTM_User, time=0.8, value=1}
          {tangentMode=RCTM_User, time=1, value=1}
         ]}
       }}
      ScaleCurve=<1 1>
     )
     Alpha=&SelectFloatFromArray(
      Direct Array Index=&SpawnGroup()
      Float Selection Array=DataInterface() {propertyValues {FloatData [1,1,1,1,1,0,1,0]}}
     )
    )
    Scale Sprite Size Mode=::NewEnumerator2
   )
   &SkeletalMeshLocation(
    Bone / Socket Index=&SelectIntFromArray(
     Direct Array Index=&SpawnGroup()
     Int Selection Array=DataInterface() {propertyValues {IntData [0,1,2,3,4,5,6,7]}}
    )
    Bone Sampling Mode=::NewEnumerator4
    MaskSpawnGroup=false
    Sampled Position Offset=&SelectVectorFromArray(
     Direct Array Index=&SpawnGroup()
     Vector Selection Array=DataInterface() {propertyValues {FloatData [<-0.4 0 0>, <-0.4 0 0>, <-0.4 0 0>, <-0.4 0 0>, <-0.4 0 0>, <-0.4 0 0>, <-0.4 0 0>, <-5 2 -1>, <-1.3205 -8.9357 -3>, <-1.3205 -8.9357 4>]}}
    )
    Skeletal Mesh=$System.SkeletalMesh
   )
   &DynamicMaterialParameters(
    Index 0 Param 1=1
    Index 0 Param 2=1
    Index 0 Param 3=&FloatFromCurve(FloatCurve=DataInterface() {propertyValues {Curve {keys [
         <at 0 0>
         {arriveTangent=0.44, interpMode=RCIM_Cubic, leaveTangent=0.44, time=0.4152, value=0.014}
         {arriveTangent=1.6861, interpMode=RCIM_Cubic, leaveTangent=1.6861, time=0.8228, value=0.3621}
         {arriveTangent=6.8718, time=1, value=1}
        ]}}})
    Index 0 Param 4=&RandomRangeFloat(Minimum=0, Maximum=33)
    Index 1 Param 1=&FloatFromCurve(
     FloatCurve=DataInterface() {propertyValues {Curve {keys [
         <at 0.216 1 RCIM_Cubic>
         {interpMode=RCIM_Cubic, tangentMode=RCTM_User, time=0.3505, value=0}
         {interpMode=RCIM_Cubic, tangentMode=RCTM_User, time=0.71, value=0}
         <at 0.8425 1 RCIM_Cubic>
        ]}}}
     CurveIndex=Linked() {linkedVariable {name=NPC.FortniteNPC.TimeOfDay_0-1, type=NiagaraFloat}}
    )
    Index 1 Param 2=&FloatFromCurve(
     FloatCurve=DataInterface() {propertyValues {Curve {keys [
         <at 0.216 0.4 RCIM_Cubic>
         {interpMode=RCIM_Cubic, tangentMode=RCTM_User, time=0.3505, value=0.94}
         {interpMode=RCIM_Cubic, tangentMode=RCTM_User, time=0.71, value=0.94}
         <at 0.8425 0.4 RCIM_Cubic>
        ]}}}
     CurveIndex=Linked() {linkedVariable {name=NPC.FortniteNPC.TimeOfDay_0-1, type=NiagaraFloat}}
    )
    Module.Index0Param0Write=true
    Module.Index0Param1Write=true
    Module.Index0Param2Write=true
    Module.Index0Param3Write=true
    Module.Index1Param0Write=true
    Module.Index1Param1Write=true
    Module.Index1Param2Write=true
    Module.Index1Param3Write=true
    Write Parameter Index 1=true
   )
   /SetVariables(Particles.Position=&AddVectorToPosition(
     Position=$Particles.Position
     Vector=&VectorFromCurve(
      VectorCurve=DataInterface() {propertyValues {
        XCurve {keys [
          {arriveTangent=0.3062, interpMode=RCIM_Cubic, leaveTangent=0.3062, tangentMode=RCTM_User, time=0, value=0}
          {arriveTangent=1.5297, tangentMode=RCTM_User, time=1, value=1}
         ]}
        YCurve {keys [
          {arriveTangent=0.3062, interpMode=RCIM_Cubic, leaveTangent=0.3062, tangentMode=RCTM_User, time=0, value=0}
          {arriveTangent=1.5297, tangentMode=RCTM_User, time=1, value=1}
         ]}
        ZCurve {keys [
          {arriveTangent=0.3062, interpMode=RCIM_Cubic, leaveTangent=0.3062, tangentMode=RCTM_User, time=0, value=0}
          {arriveTangent=1.5297, tangentMode=RCTM_User, time=1, value=1}
         ]}
       }}
      CurveIndex=$Particles.Age
      Scale Curve=&Multiply_Vector(
       A=<1 1 1>
       B=&Add_Vector(
        A=&RotateVectorByQuaternion(
         Vector To Rotate=&RandomRangeVector(Minimum=<11 0 0>, Maximum=<16 0 0>)
         Quaternion=$Particles.SkeletalMeshLocation.SampledRotation
        )
        B=<0 0 5>
       )
      )
     )
    )) {isSetParameters=true}
   &SpriteFacingAndAlignment(
    Sprite Alignment=&VectorFromCurve(
     VectorCurve=DataInterface() {propertyValues {
       XCurve {keys [<at 0 0 RCIM_Cubic>, {arriveTangent=2.1915, tangentMode=RCTM_User, time=1, value=1}]}
       YCurve {keys [<at 0 0 RCIM_Cubic>, {arriveTangent=2.1915, tangentMode=RCTM_User, time=1, value=1}]}
       ZCurve {keys [<at 0 0 RCIM_Cubic>, {arriveTangent=2.1915, tangentMode=RCTM_User, time=1, value=1}]}
      }}
     CurveIndex=$Particles.Age
     Scale Curve=&Multiply_Vector(
      A=<1 1 1>
      B=&Add_Vector(
       A=&RotateVectorByQuaternion(
        Vector To Rotate=&RandomRangeVector(Minimum=<11 0 0>, Maximum=<20 0 0>)
        Quaternion=$Particles.SkeletalMeshLocation.SampledRotation
       )
       B=<0 0 5>
      )
     )
    )
    WriteAlignment=true
    WriteFacing=false
   )
   /SetVariables(Particles.RimLightDynamic:Vector4f=&MakeVector4_FromLinearColor(LinearColor=&LerpLinearColors(
      StartColor=&ColorFromCurve(
       ColorCurve=DataInterface() {propertyValues {
         AlphaCurve {keys [<at 0 1>,<at 0.125 1>,<at 0.25 1>,<at 0.375 1>,<at 0.5 1>,<at 0.625 1>,<at 0.75 1>,<at 0.875 1>,<at 1 1>]}
         BlueCurve {keys [<at 0 0.01>,<at 0.041 0.01>,<at 0.083 0.01>,<at 0.125 0.01>,<at 0.166 0.0077>,<at 0.208 0.0045>,<at 0.25 0.0092>,<at 0.291 0.0136>,<at 0.333 0.03>,<at 0.375 0.04>,<at 0.416 0.0525>,<at 0.458 0.0525>,<at 0.5 0.0525>,<at 0.541 0.0525>,<at 0.583 0.0525>,<at 0.625 0.04>,<at 0.666 0.03>,<at 0.708 0.0152>,<at 0.75 0.0153>,<at 0.791 0.0079>,<at 0.833 0.0025>,<at 0.875 0.0031>,<at 0.916 0.01>,<at 0.958 0.01>,<at 1 0.01>]}
         GreenCurve {keys [<at 0 0.0068>,<at 0.041 0.0064>,<at 0.083 0.0064>,<at 0.125 0.0071>,<at 0.166 0.0056>,<at 0.208 0.0041>,<at 0.25 0.0103>,<at 0.291 0.0186>,<at 0.333 0.0156>,<at 0.375 0.0208>,<at 0.416 0.0272>,<at 0.458 0.0272>,<at 0.5 0.0272>,<at 0.541 0.0272>,<at 0.583 0.0272>,<at 0.625 0.0208>,<at 0.666 0.0156>,<at 0.708 0.0192>,<at 0.75 0.0153>,<at 0.791 0.0092>,<at 0.833 0.0051>,<at 0.875 0.0062>,<at 0.916 0.0064>,<at 0.958 0.0063>,<at 1 0.0068>]}
         RedCurve {keys [<at 0 0.0033>,<at 0.041 0.0029>,<at 0.083 0.0032>,<at 0.125 0.0036>,<at 0.166 0.01>,<at 0.208 0.01>,<at 0.25 0.02>,<at 0.291 0.03>,<at 0.333 0.0086>,<at 0.375 0.0115>,<at 0.416 0.0151>,<at 0.458 0.0151>,<at 0.5 0.0151>,<at 0.541 0.0151>,<at 0.583 0.0151>,<at 0.625 0.0115>,<at 0.666 0.0086>,<at 0.708 0.03>,<at 0.75 0.03>,<at 0.791 0.02>,<at 0.833 0.01>,<at 0.875 0.01>,<at 0.916 0.0022>,<at 0.958 0.0019>,<at 1 0.0033>]}
        }}
       CurveIndex=Linked() {linkedVariable {name=NPC.FortniteNPC.TimeOfDay_0-1, type=NiagaraFloat}}
      )
      EndColor=<rgba 0.0151 0.0272 0.0525 1>
      LerpFactor=&FN_FrontEndFloatModifier(Gameplay=0, Front End=1)
     ))) {isSetParameters=true}
   /SetVariables(Particles.HighlightDynamic:Vector4f=&MakeVector4_FromLinearColor(LinearColor=&LerpLinearColors(
      StartColor=&ColorFromCurve(
       ColorCurve=DataInterface() {propertyValues {
         AlphaCurve {keys [<at 0 1>,<at 0.125 1>,<at 0.25 1>,<at 0.375 1>,<at 0.5 1>,<at 0.625 1>,<at 0.75 1>,<at 0.875 1>,<at 1 1>]}
         BlueCurve {keys [<at 0 0.9216>,<at 0.041 0.8963>,<at 0.083 0.8308>,<at 0.125 0.7835>,<at 0.1865 0.2611>,<at 0.25 0.3013>,<at 0.333 0.3128>,<at 0.4094 0.2806>,<at 0.4889 0.2616>,<at 0.583 0.2814>,<at 0.666 0.2801>,<at 0.75 0.2471>,<at 0.791 0.2118>,<at 0.833 0.2338>,<at 0.875 0.2473>,<at 0.916 0.8388>,<at 0.958 0.9047>,<at 1 0.9216>]}
         GreenCurve {keys [<at 0 0.5616>,<at 0.041 0.5276>,<at 0.083 0.4795>,<at 0.125 0.5099>,<at 0.1865 0.2253>,<at 0.25 0.3543>,<at 0.333 0.4809>,<at 0.4094 0.4334>,<at 0.4889 0.389>,<at 0.583 0.2814>,<at 0.666 0.4785>,<at 0.75 0.2471>,<at 0.791 0.2686>,<at 0.833 0.4442>,<at 0.875 0.5188>,<at 0.916 0.5292>,<at 0.958 0.5688>,<at 1 0.6308>]}
         RedCurve {keys [<at 0 0.1584>,<at 0.041 0.1774>,<at 0.083 0.1644>,<at 0.125 0.1755>,<at 0.1865 0.6867>,<at 0.25 0.8148>,<at 0.333 0.8963>,<at 0.4094 0.9131>,<at 0.4889 0.9131>,<at 0.583 0.8714>,<at 0.666 0.8963>,<at 0.75 0.7529>,<at 0.791 0.7529>,<at 0.833 0.8469>,<at 0.875 0.8632>,<at 0.916 0.1791>,<at 0.958 0.1696>,<at 1 0.305>]}
        }}
       CurveIndex=Linked() {linkedVariable {name=NPC.FortniteNPC.TimeOfDay_0-1, type=NiagaraFloat}}
      )
      EndColor=<rgba 0.7 0.7 0.7 1>
      LerpFactor=&FN_FrontEndFloatModifier(Gameplay=0, Front End=1)
     ))) {isSetParameters=true}
  ]
  rendererClasses [&NiagaraSpriteRendererProperties]
  renderers [{rendererClass=&NiagaraSpriteRendererProperties}]
 }}
summary {
 systemName=NS_Character_MissusMind_Goop_Idle_Body
 emitters [{emitterName=Goop, rendererClasses [&NiagaraSpriteRendererProperties]}]
 userVariables [
  User.bisFootHidden {type=NiagaraBool}
  User.bIsFrontEnd {type=NiagaraBool}
  User.bIsFrontEndPreview {type=NiagaraBool}
  User.bisHandHidden {type=NiagaraBool}
 ]
}
