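# Character material for the Missus Mind Goop female body, layering a twisted goo effect onto the standard character shading.
# It runs the shared 2023 character parent PreFX function, feeds the result through a pre-skinned twisted goo function, then finishes with the parent PostFX function.
# The goo's alpha comes from a GoopArea channel of the Tech Art Mask, or covers everything via an ApplyToAll static switch.
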Material /BRCosmetics/Characters/Player/Female/Medium/Bodies/F_MED_MissusMind_Goop/Materials/M_MissusMind_Goop
graph {
 properties {
  bUseMaterialAttributes=true
  bUsedWithMeshDeformer=true
  bUsedWithMorphTargets=true
  bUsedWithSkeletalMesh=true
  pixelDepthOffsetMode=PDOM_Legacy
  shadingModel=MSM_FromMaterialExpression
  usedShadingModels=Default Lit | Subsurface | Clear Coat | Subsurface Profile | Cloth
 }
 nodes [
  n1 = MaterialFunctionCall() {materialFunction=&MF_CharacterParent_2023_PreFX}
  n2 = Constant() {r=1}
  n3 = ChannelMaskParameter(None=@n1.Tech Art Mask) {group=Goop, parameterName=GoopArea}
  MaterialFunctionCall(
   Material=MaterialFunctionCall!(In=@n1.Result, Alpha=StaticSwitchParameter(True=@n2, False=@n3.R) {group=Goop, parameterName=ApplyToAll}) {materialFunction=&/MF_TwistedGoo_Preskinned}.Result
   M=@n1.M
   Pre Skinned Local Position=@n1.M
   Distance Blend=@n1.M
   Fresnel=@n1.M
  ) {materialFunction=&MF_CharacterParent_2023_PostFX}
 ]
}
