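# Physics asset for the MissusMind Goop medium female player body, covering the torso and four dynamic tube bones (upper and lower, left and right).
# Capsules form a pelvis-to-spine_04 chain, and four small tube master bodies all attach to spine_02, with swing and twist free on these joints.
# Each larger, offset tube capsule is angularly locked to its master, so each pair keeps a fixed relative orientation.
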PhysicsAsset /BRCosmetics/Characters/Player/Female/Medium/Bodies/F_MED_MissusMind_Goop/Meshes/F_MED_MissusMind_Goop_Physics
bodies {
 dyn_tube_low_l [{shapeType=Capsule, length=1.4235, radius=1.5721, shapeName=dyn_tube_low_l_capsule, center=<2.3 0 0>}]
 dyn_tube_low_mstr_l [{shapeType=Capsule, length=0.505, radius=0.505, shapeName=dyn_tube_low_mstr_l_capsule}]
 dyn_tube_low_mstr_r [{shapeType=Capsule, length=0.505, radius=0.505, shapeName=dyn_tube_low_mstr_r_capsule}]
 dyn_tube_low_r [{shapeType=Capsule, length=1.7095, radius=1.5981, shapeName=dyn_tube_low_r_capsule, center=<2.3 0 0>}]
 dyn_tube_up_l [{shapeType=Capsule, length=1.7906, radius=1.5096, shapeName=dyn_tube_up_l_capsule, center=<2.3 0 0>}]
 dyn_tube_up_mstr_l [{shapeType=Capsule, length=0.505, radius=0.505, shapeName=dyn_tube_up_mstr_l_capsule}]
 dyn_tube_up_mstr_r [{shapeType=Capsule, length=0.505, radius=0.505, shapeName=dyn_tube_up_mstr_r_capsule}]
 dyn_tube_up_r [{shapeType=Capsule, length=1.5547, radius=1.8098, shapeName=dyn_tube_up_r_capsule, center=<2.3 0 0>}]
 pelvis [{shapeType=Capsule, length=8.8149, radius=8.7805, shapeName=pelvis_capsule, center=<0 -1.6997 0>}]
 spine_01 [{shapeType=Capsule, length=6.1711, radius=7.7679, shapeName=spine_01_capsule}]
 spine_02 [{shapeType=Capsule, length=1.2998, radius=7.8863, shapeName=spine_02_capsule}]
 spine_03 [{shapeType=Capsule, length=5.319, radius=6.5722, shapeName=spine_03_capsule}]
 spine_04 [{shapeType=Capsule, length=5.9807, radius=8.1772, shapeName=spine_04_capsule}]
}
constraints [
 {
  bone1Name=spine_01
  bone2Name=pelvis
  swing1LimitDegrees=45
  swing1Motion=Free
  swing2LimitDegrees=45
  swing2Motion=Free
  twistLimitDegrees=45
  twistMotion=Free
 }
 {
  bone1Name=spine_02
  bone2Name=spine_01
  swing1LimitDegrees=45
  swing1Motion=Free
  swing2LimitDegrees=45
  swing2Motion=Free
  twistLimitDegrees=45
  twistMotion=Free
 }
 {
  bone1Name=dyn_tube_low_l
  bone2Name=dyn_tube_low_mstr_l
  swing1LimitDegrees=4
  swing1Motion=Locked
  swing2LimitDegrees=4
  swing2Motion=Locked
  twistLimitDegrees=2
  twistMotion=Locked
 }
 {
  bone1Name=dyn_tube_low_r
  bone2Name=dyn_tube_low_mstr_r
  swing1LimitDegrees=4
  swing1Motion=Locked
  swing2LimitDegrees=4
  swing2Motion=Locked
  twistLimitDegrees=2
  twistMotion=Locked
 }
 {
  bone1Name=dyn_tube_up_l
  bone2Name=dyn_tube_up_mstr_l
  swing1LimitDegrees=4
  swing1Motion=Locked
  swing2LimitDegrees=4
  swing2Motion=Locked
  twistLimitDegrees=2
  twistMotion=Locked
 }
 {
  bone1Name=dyn_tube_up_r
  bone2Name=dyn_tube_up_mstr_r
  swing1LimitDegrees=4
  swing1Motion=Locked
  swing2LimitDegrees=4
  swing2Motion=Locked
  twistLimitDegrees=2
  twistMotion=Locked
 }
 {
  bone1Name=spine_03
  bone2Name=spine_02
  swing1LimitDegrees=45
  swing1Motion=Free
  swing2LimitDegrees=45
  swing2Motion=Free
  twistLimitDegrees=45
  twistMotion=Free
 }
 {
  bone1Name=spine_04
  bone2Name=spine_03
  swing1LimitDegrees=45
  swing1Motion=Free
  swing2LimitDegrees=45
  swing2Motion=Free
  twistLimitDegrees=45
  twistMotion=Free
 }
 {
  bone1Name=dyn_tube_low_mstr_l
  bone2Name=spine_02
  swing1LimitDegrees=45
  swing1Motion=Free
  swing2LimitDegrees=45
  swing2Motion=Free
  twistLimitDegrees=45
  twistMotion=Free
 }
 {
  bone1Name=dyn_tube_low_mstr_r
  bone2Name=spine_02
  swing1LimitDegrees=45
  swing1Motion=Free
  swing2LimitDegrees=45
  swing2Motion=Free
  twistLimitDegrees=45
  twistMotion=Free
 }
 {
  bone1Name=dyn_tube_up_mstr_l
  bone2Name=spine_02
  swing1LimitDegrees=45
  swing1Motion=Free
  swing2LimitDegrees=45
  swing2Motion=Free
  twistLimitDegrees=45
  twistMotion=Free
 }
 {
  bone1Name=dyn_tube_up_mstr_r
  bone2Name=spine_02
  swing1LimitDegrees=45
  swing1Motion=Free
  swing2LimitDegrees=45
  swing2Motion=Free
  twistLimitDegrees=45
  twistMotion=Free
 }
]
