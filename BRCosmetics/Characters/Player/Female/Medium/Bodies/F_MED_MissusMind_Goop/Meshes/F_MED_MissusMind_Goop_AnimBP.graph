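# Anim Blueprint for the MissusMind Goop character body, animating dangling goop strands on the hands, shoulder, elbow and heel.
# The copied parent pose passes through Control Rig and RigidBody with overridden gravity, then additive bone edits and constraints pinning FX bones to strand tips.
# Each update generates Perlin rotators, paced by sprint and skydive state.
# Yaw-derived X-scale vectors stretch strands, but drive translation on elbow and heel.
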AnimBlueprint /BRCosmetics/Characters/Player/Female/Medium/Bodies/F_MED_MissusMind_Goop/Meshes/F_MED_MissusMind_Goop_AnimBP
target_skeleton=&/F_MED_MissusMind_Goop_Skeleton
graphs [
 AnimGraph {nodes [
   n1 = Control Rig(Source=Copy Pose From Mesh() {bUseAttachedParent=true}) {
    alphaBoolBlend {blendOption=Linear}
    controlRigAssetReference {blueprintRigClass=&/F_MED_MissusMind_Goop_CtrlRig.F_MED_MissusMind_Goop_CtrlRig_C}
   }
   n2 = RigidBody!(ComponentPose=Local To Component(LocalPose=@n1), Alpha, OverrideWorldGravity:Vector=Get Gravity Override(), ExternalForce) -> (Pose) {
    bOverrideWorldGravity=true
    overridePhysicsAsset=&/F_MED_MissusMind_Goop_Physics
    simulationSpace=BaseBoneSpace
    alphaBoolBlend {blendOption=Linear}
    baseBoneRef {boneName=spine_02}
    componentAppliedLinearAccClamp=<5000 5000 5000>
    componentLinearAccScale=<0.4 0.4 0.4>
    componentLinearVelScale=<0.5 0.5 0.5>
    simSpaceSettings {maxAngularAcceleration=300, maxAngularVelocity=10, maxLinearAcceleration=500, maxLinearVelocity=300, worldAlpha=0.7}
   }
   n3 = "Transform (Modify) Bone - Bone: dyn_hand_goop_01_l"(ComponentPose=@n2, Alpha=1, Translation="0,0,0", Rotation="0,0,0", Scale="1,1,1") {
    rotationMode=BMM_Additive
    rotationSpace=BCS_BoneSpace
    scaleMode=BMM_Additive
    scaleSpace=BCS_BoneSpace
    alphaBoolBlend {blendOption=Linear}
    boneToModify {boneName=dyn_hand_goop_01_l}
   }
   n4 = Get HairRotator_01() -> (HairRotator_01:Rotator)
   n5 = Get XScale() -> (XScale:Vector)
   n6 = "Transform (Modify) Bone - Bone: dyn_hand_goop_02_l"(ComponentPose=@n3, Alpha=1, Translation="0,0,0", Rotation=@n4, Scale=@n5) {
    rotationMode=BMM_Additive
    rotationSpace=BCS_BoneSpace
    scaleMode=BMM_Additive
    scaleSpace=BCS_BoneSpace
    alphaBoolBlend {blendOption=Linear}
    boneToModify {boneName=dyn_hand_goop_02_l}
   }
   n7 = Get HairRotator_02() -> (HairRotator_02:Rotator)
   n8 = Get XScale_2() -> (XScale_2:Vector)
   n9 = "Transform (Modify) Bone - Bone: dyn_hand_goop_03_l"(ComponentPose=@n6, Alpha=1, Translation="0,0,0", Rotation=@n7, Scale=@n8) {
    rotationMode=BMM_Additive
    rotationSpace=BCS_BoneSpace
    scaleMode=BMM_Additive
    scaleSpace=BCS_BoneSpace
    alphaBoolBlend {blendOption=Linear}
    boneToModify {boneName=dyn_hand_goop_03_l}
   }
   n10 = "Transform (Modify) Bone - Bone: dyn_shld_goop_01_l"(ComponentPose=@n9, Alpha=1, Translation="0,0,0", Rotation="0,0,0", Scale="1,1,1") {
    rotationMode=BMM_Additive
    rotationSpace=BCS_BoneSpace
    scaleMode=BMM_Additive
    scaleSpace=BCS_BoneSpace
    alphaBoolBlend {blendOption=Linear}
    boneToModify {boneName=dyn_shld_goop_01_l}
   }
   n11 = "Transform (Modify) Bone - Bone: dyn_shld_goop_02_l"(ComponentPose=@n10, Alpha=1, Translation="0,0,0", Rotation="0,0,0", Scale=@n8) {
    rotationMode=BMM_Additive
    rotationSpace=BCS_BoneSpace
    scaleMode=BMM_Additive
    scaleSpace=BCS_BoneSpace
    alphaBoolBlend {blendOption=Linear}
    boneToModify {boneName=dyn_shld_goop_02_l}
   }
   n12 = "Transform (Modify) Bone - Bone: dyn_shld_goop_03_l"(ComponentPose=@n11, Alpha=1, Translation="0,0,0", Rotation="0,0,0", Scale=@n5) {
    rotationMode=BMM_Additive
    rotationSpace=BCS_BoneSpace
    scaleMode=BMM_Additive
    scaleSpace=BCS_BoneSpace
    alphaBoolBlend {blendOption=Linear}
    boneToModify {boneName=dyn_shld_goop_03_l}
   }
   n13 = "Transform (Modify) Bone - Bone: dyn_shld_goop_sml_01_l"(ComponentPose=@n12, Alpha=1, Translation="0,0,0", Rotation="0,0,0", Scale="1,1,1") {
    rotationMode=BMM_Additive
    rotationSpace=BCS_BoneSpace
    scaleMode=BMM_Additive
    scaleSpace=BCS_BoneSpace
    alphaBoolBlend {blendOption=Linear}
    boneToModify {boneName=dyn_shld_goop_sml_01_l}
   }
   n14 = Get HairRotator_03() -> (HairRotator_03:Rotator)
   n15 = "Transform (Modify) Bone - Bone: dyn_shld_goop_sml_02_l"(ComponentPose=@n13, Alpha=1, Translation="0,0,0", Rotation=@n14, Scale=@n5) {
    rotationMode=BMM_Additive
    rotationSpace=BCS_BoneSpace
    scaleMode=BMM_Additive
    scaleSpace=BCS_BoneSpace
    alphaBoolBlend {blendOption=Linear}
    boneToModify {boneName=dyn_shld_goop_sml_02_l}
   }
   n16 = "Transform (Modify) Bone - Bone: dyn_shld_goop_sml_03_l"(ComponentPose=@n15, Alpha=1, Translation="0,0,0", Rotation=@n4, Scale=@n8) {
    rotationMode=BMM_Additive
    rotationSpace=BCS_BoneSpace
    scaleMode=BMM_Additive
    scaleSpace=BCS_BoneSpace
    alphaBoolBlend {blendOption=Linear}
    boneToModify {boneName=dyn_shld_goop_sml_03_l}
   }
   n17 = "Transform (Modify) Bone - Bone: dyn_index_goop_01_r"(ComponentPose=@n16, Alpha=1, Translation="0,0,0", Rotation="0,0,0", Scale="1,1,1") {
    rotationMode=BMM_Additive
    rotationSpace=BCS_BoneSpace
    scaleMode=BMM_Additive
    scaleSpace=BCS_BoneSpace
    alphaBoolBlend {blendOption=Linear}
    boneToModify {boneName=dyn_index_goop_01_r}
   }
   n18 = "Transform (Modify) Bone - Bone: dyn_index_goop_02_r"(ComponentPose=@n17, Alpha=1, Translation="0,0,0", Rotation=@n7, Scale=@n5) {
    rotationMode=BMM_Additive
    rotationSpace=BCS_BoneSpace
    scaleMode=BMM_Additive
    scaleSpace=BCS_BoneSpace
    alphaBoolBlend {blendOption=Linear}
    boneToModify {boneName=dyn_index_goop_02_r}
   }
   n19 = "Transform (Modify) Bone - Bone: dyn_index_goop_03_r"(ComponentPose=@n18, Alpha=1, Translation="0,0,0", Rotation=@n14, Scale=@n8) {
    rotationMode=BMM_Additive
    rotationSpace=BCS_BoneSpace
    scaleMode=BMM_Additive
    scaleSpace=BCS_BoneSpace
    alphaBoolBlend {blendOption=Linear}
    boneToModify {boneName=dyn_index_goop_03_r}
   }
   n20 = "Transform (Modify) Bone - Bone: dyn_pinky_goop_01_r"(ComponentPose=@n19, Alpha=1, Translation="0,0,0", Rotation="0,0,0", Scale="1,1,1") {
    rotationMode=BMM_Additive
    rotationSpace=BCS_BoneSpace
    scaleMode=BMM_Additive
    scaleSpace=BCS_BoneSpace
    alphaBoolBlend {blendOption=Linear}
    boneToModify {boneName=dyn_pinky_goop_01_r}
   }
   n21 = "Transform (Modify) Bone - Bone: dyn_pinky_goop_02_r"(ComponentPose=@n20, Alpha=1, Translation="0,0,0", Rotation=@n4, Scale=@n5) {
    rotationMode=BMM_Additive
    rotationSpace=BCS_BoneSpace
    scaleMode=BMM_Additive
    scaleSpace=BCS_BoneSpace
    alphaBoolBlend {blendOption=Linear}
    boneToModify {boneName=dyn_pinky_goop_02_r}
   }
   n22 = "Transform (Modify) Bone - Bone: dyn_pinky_goop_03_r"(ComponentPose=@n21, Alpha=1, Translation="0,0,0", Rotation=@n7, Scale=@n8) {
    rotationMode=BMM_Additive
    rotationSpace=BCS_BoneSpace
    scaleMode=BMM_Additive
    scaleSpace=BCS_BoneSpace
    alphaBoolBlend {blendOption=Linear}
    boneToModify {boneName=dyn_pinky_goop_03_r}
   }
   n23 = "Transform (Modify) Bone - Bone: dyn_elb_goop_01_r"(ComponentPose=@n22, Alpha=1, Translation="0,0,0", Rotation="0,0,0", Scale="1,1,1") {
    rotationMode=BMM_Additive
    rotationSpace=BCS_BoneSpace
    scaleMode=BMM_Additive
    scaleSpace=BCS_BoneSpace
    alphaBoolBlend {blendOption=Linear}
    boneToModify {boneName=dyn_elb_goop_01_r}
   }
   n24 = "Transform (Modify) Bone - Bone: dyn_elb_goop_02_r"(ComponentPose=@n23, Alpha=1, Translation=@n8, Rotation=@n14, Scale="1,1,1") {
    rotationMode=BMM_Additive
    rotationSpace=BCS_BoneSpace
    scaleMode=BMM_Additive
    scaleSpace=BCS_BoneSpace
    alphaBoolBlend {blendOption=Linear}
    boneToModify {boneName=dyn_elb_goop_02_r}
   }
   n25 = "Transform (Modify) Bone - Bone: dyn_elb_goop_03_r"(ComponentPose=@n24, Alpha=1, Translation=@n5, Rotation=@n4, Scale="1,1,1") {
    rotationMode=BMM_Additive
    rotationSpace=BCS_BoneSpace
    scaleMode=BMM_Additive
    scaleSpace=BCS_BoneSpace
    alphaBoolBlend {blendOption=Linear}
    boneToModify {boneName=dyn_elb_goop_03_r}
   }
   n26 = "Transform (Modify) Bone - Bone: dyn_heel_goop_01_l"(ComponentPose=@n25, Alpha=1, Translation="0,0,0", Rotation="0,0,0", Scale="1,1,1") {
    rotationMode=BMM_Additive
    rotationSpace=BCS_BoneSpace
    scaleMode=BMM_Additive
    scaleSpace=BCS_BoneSpace
    alphaBoolBlend {blendOption=Linear}
    boneToModify {boneName=dyn_heel_goop_01_l}
   }
   n27 = "Transform (Modify) Bone - Bone: dyn_heel_goop_02_l"(ComponentPose=@n26, Alpha=1, Translation=@n5, Rotation=@n7, Scale="1,1,1") {
    rotationMode=BMM_Additive
    rotationSpace=BCS_BoneSpace
    scaleMode=BMM_Additive
    scaleSpace=BCS_BoneSpace
    alphaBoolBlend {blendOption=Linear}
    boneToModify {boneName=dyn_heel_goop_02_l}
   }
   n28 = "Transform (Modify) Bone - Bone: dyn_heel_goop_03_l"(ComponentPose=@n27, Alpha=1, Translation=@n8, Rotation=@n14, Scale="1,1,1") {
    rotationMode=BMM_Additive
    rotationSpace=BCS_BoneSpace
    scaleMode=BMM_Additive
    scaleSpace=BCS_BoneSpace
    alphaBoolBlend {blendOption=Linear}
    boneToModify {boneName=dyn_heel_goop_03_l}
   }
   n29 = "Constraint - FX_hand_goop_l (P)"(ComponentPose=@n28, Alpha=1, ConstraintWeights_0=1) {
    alphaBoolBlend {blendOption=Linear}
    boneToModify {boneName=FX_hand_goop_l}
    constraintSetup [{targetBone {boneName=dyn_hand_goop_03_l}}]
    constraintWeights [1]
   }
   n30 = "Constraint - FX_shld_goop_l (P)"(ComponentPose=@n29, Alpha=1, ConstraintWeights_0=1) {
    alphaBoolBlend {blendOption=Linear}
    boneToModify {boneName=FX_shld_goop_l}
    constraintSetup [{targetBone {boneName=dyn_shld_goop_03_l}}]
    constraintWeights [1]
   }
   n31 = "Constraint - FX_shld_goop_sml_l (P)"(ComponentPose=@n30, Alpha=1, ConstraintWeights_0=1) {
    alphaBoolBlend {blendOption=Linear}
    boneToModify {boneName=FX_shld_goop_sml_l}
    constraintSetup [{targetBone {boneName=dyn_shld_goop_sml_03_l}}]
    constraintWeights [1]
   }
   n32 = "Constraint - FX_elb_goop_r (P)"(ComponentPose=@n31, Alpha=1, ConstraintWeights_0=1) {
    alphaBoolBlend {blendOption=Linear}
    boneToModify {boneName=FX_elb_goop_r}
    constraintSetup [{targetBone {boneName=dyn_elb_goop_03_r}}]
    constraintWeights [1]
   }
   n33 = "Constraint - FX_index_goop_r (P)"(ComponentPose=@n32, Alpha=1, ConstraintWeights_0=1) {
    alphaBoolBlend {blendOption=Linear}
    boneToModify {boneName=FX_index_goop_r}
    constraintSetup [{targetBone {boneName=dyn_index_goop_03_r}}]
    constraintWeights [1]
   }
   n34 = "Constraint - FX_pinky_goop_r (P)"(ComponentPose=@n33, Alpha=1, ConstraintWeights_0=1) {
    alphaBoolBlend {blendOption=Linear}
    boneToModify {boneName=FX_pinky_goop_r}
    constraintSetup [{targetBone {boneName=dyn_pinky_goop_03_r}}]
    constraintWeights [1]
   }
   n35 = "Constraint - FX_heel_goop_l (P)"(ComponentPose=@n34, Alpha=1, ConstraintWeights_0=1) {
    alphaBoolBlend {blendOption=Linear}
    boneToModify {boneName=FX_heel_goop_l}
    constraintSetup [{targetBone {boneName=dyn_heel_goop_03_l}}]
    constraintWeights [1]
   }
   Output Pose(Result=Component To Local(ComponentPose=@n35)) {layerGroup=None}
  ]}
 EventGraph {nodes [
   n1 = Event BlueprintUpdateAnimation()
   > n2 = Sequence!(execute) -> (then_0, then_1, then_2:Exec)
   > n3 = Gravity Override State Function DEPRECATED(
    CharacterPartAnimInstance=Self-Reference()
    GameTimeInSeconds=@n1.DeltaTimeX
    GravityOverrideParams=Get Gravity Override Params()
    RBAN Params=Get RBAN Params()
   )
   > Set Gravity Override(Gravity Override=@n3)
   n4 = Branch(execute=@n2.then_1, Condition=OR Boolean(A=Get bIsSkyDiving(), B=Get bIsTacticalSprinting()))
   > Set shakeFrequency(shakeFrequency:double=0.2) -> (Output_Get:double)
   > Set shakeFrequency2(shakeFrequency2:double=0.3) -> (Output_Get:double)
   > n5 = Branch(Condition=Get bIsSkydiveDiveMode())
   > Set shakeFrequency(shakeFrequency=0.7)
   > Set shakeFrequency2(shakeFrequency2=0.8)
   n6 = Branch(execute=@n5.else, Condition=Get bIsSkydiveFloating())
   Set shakeFrequency(execute=@n6.else, shakeFrequency=0.6)
   > Set shakeFrequency2(shakeFrequency2=0.7)
   Set shakeFrequency(execute=@n4.else, shakeFrequency=0.07)
   > Set shakeFrequency2(shakeFrequency2=0.17)
   n7 = GetGameTimeInSeconds()
   n8 = float * float(A=@n7, B=Get shakeFrequency() -> (shakeFrequency:double)) {is=title K2Node_PromotableOperator}
   n9 = RandomRotatorFunction!(
    execute=@n2.then_2
    GameTimeInSeconds:double=@n8
    PerlinScale:double=0.7
    Frequency:Rotator="18,20,0"
    Amplitude:Rotator="9,9,0"
    Offset:Rotator="0, 0, 0"
    TimeOffset:double=0
   ) -> (then, RandomRotator:Rotator)
   > Set HairRotator_01(HairRotator_01:Rotator=@n9) -> (Output_Get:Rotator)
   n10 = float * float(A=@n7, B=Get shakeFrequency2() -> (shakeFrequency2:double))
   > n11 = RandomRotatorFunction(GameTimeInSeconds=@n10, PerlinScale=0.6, Frequency="20,18,0", Amplitude="8,8,0", Offset="0,3, 0", TimeOffset=0.5)
   > Set HairRotator_02(HairRotator_02:Rotator=@n11) -> (Output_Get:Rotator)
   > n12 = RandomRotatorFunction(GameTimeInSeconds=@n8, PerlinScale=0.8, Frequency="24,18,0", Amplitude="10,10,0", Offset="0, 0, 0", TimeOffset=0.2)
   > Set HairRotator_03(HairRotator_03:Rotator=@n12) -> (Output_Get:Rotator)
   > n13 = MakeScaleFromRotation!(execute, self, InRot:Rotator=@n12, InScaleRange:Vector 2D Structure=Get ScaleRange() -> (ScaleRange:Vector 2D Structure)) -> (then, OutScale:Vector)
   > Set XScale_2(XScale_2:Vector=@n13) -> (Output_Get:Vector)
   > n14 = RandomRotatorFunction(GameTimeInSeconds=@n10, PerlinScale=1, Frequency="17,19,0", Amplitude="8,6,0", Offset="0, 0, 0", TimeOffset=0.2)
   > Set HairRotator_04(HairRotator_04:Rotator=@n14) -> (Output_Get:Rotator)
   > n15 = MakeScaleFromRotation(InRot=@n14, InScaleRange=Get ScaleRange_2() -> (ScaleRange_2:Vector 2D Structure))
   > Set XScale(XScale:Vector=@n15) -> (Output_Get:Vector)
  ]}
 MakeScaleFromRotation {nodes [
   n1 = MakeScaleFromRotation!() -> (then, InRot:Rotator=~, InScaleRange:Vector 2D Structure) {is=title K2Node_FunctionEntry}
   n2 = BreakRotator(InRot=@n1)
   > Return Node!(execute, OutScale:Vector=MakeVector(X=Remap Float(Input=@n2.Yaw, OldRange=(X=-10,Y=10), NewRange=@n1.InScaleRange) -> (Output=~), Y=1, Z=1))
  ]}
]
variables [
 Gravity Override {type=Vector}
 Gravity Override Params {type=GravityOverrideParamsStruct}
 HairRotator_01 {type=Rotator}
 HairRotator_02 {type=Rotator}
 HairRotator_03 {type=Rotator}
 HairRotator_04 {type=Rotator}
 RBAN Params {
  type=RBANParamsStruct
  default {
   dBNO {simSpaceSettings {
     maxAngularAcceleration=1000
     maxAngularVelocity=3
     maxLinearAcceleration=300
     maxLinearVelocity=700
     velocityScaleZ=0
     externalLinearDragV=<0 0.5 0>
    }}
   dBNO_Carried {simSpaceSettings {
     maxAngularAcceleration=1000
     maxAngularVelocity=3
     maxLinearAcceleration=300
     maxLinearVelocity=700
     velocityScaleZ=0
     externalLinearDragV=<0 0.5 0>
    }}
   emoteOrMelee {simSpaceSettings {
     maxAngularAcceleration=1000
     maxAngularVelocity=3
     maxLinearAcceleration=300
     maxLinearVelocity=700
     velocityScaleZ=0
     externalLinearDragV=<0 0.5 0>
    }}
   front_End {simSpaceSettings {
     maxAngularAcceleration=1000
     maxAngularVelocity=3
     maxLinearAcceleration=300
     maxLinearVelocity=700
     velocityScaleZ=0
     externalLinearDragV=<0 0.5 0>
    }}
   onGround_Crouch_Moving {simSpaceSettings {
     maxAngularAcceleration=1000
     maxAngularVelocity=3
     maxLinearAcceleration=300
     maxLinearVelocity=700
     velocityScaleZ=0
     externalLinearDragV=<0 0.5 0>
    }}
   onGround_Falling {simSpaceSettings {
     maxAngularAcceleration=1000
     maxAngularVelocity=3
     maxLinearAcceleration=300
     maxLinearVelocity=700
     velocityScaleZ=0
     externalLinearDragV=<0 0.5 0>
    }}
   onGround_Moving {simSpaceSettings {
     maxAngularAcceleration=1000
     maxAngularVelocity=3
     maxLinearAcceleration=300
     maxLinearVelocity=700
     velocityScaleZ=0
     externalLinearDragV=<0 0.5 0>
    }}
   onGround_Standing {simSpaceSettings {
     maxAngularAcceleration=1000
     maxAngularVelocity=3
     maxLinearAcceleration=300
     maxLinearVelocity=700
     velocityScaleZ=0
     externalLinearDragV=<0 0.5 0>
    }}
   rail_Grinding {simSpaceSettings {
     maxAngularAcceleration=1000
     maxAngularVelocity=3
     maxLinearAcceleration=300
     maxLinearVelocity=700
     velocityScaleZ=0
     worldAlpha=0.1
     externalLinearDragV=<0 0.5 0>
    }}
   riding_Moving {simSpaceSettings {
     maxAngularAcceleration=1000
     maxAngularVelocity=3
     maxLinearAcceleration=300
     maxLinearVelocity=700
     velocityScaleZ=0
     externalLinearDragV=<0 0.5 0>
    }}
   riding_Moving_Targeting {simSpaceSettings {
     maxAngularAcceleration=1000
     maxAngularVelocity=3
     maxLinearAcceleration=300
     maxLinearVelocity=700
     velocityScaleZ=0
     externalLinearDragV=<0 0.5 0>
    }}
   riding_Standing {simSpaceSettings {
     maxAngularAcceleration=1000
     maxAngularVelocity=3
     maxLinearAcceleration=300
     maxLinearVelocity=700
     velocityScaleZ=0
     externalLinearDragV=<0 0.5 0>
    }}
   skydive_DivingDown {simSpaceSettings {
     maxAngularAcceleration=1000
     maxAngularVelocity=3
     maxLinearAcceleration=300
     maxLinearVelocity=700
     velocityScaleZ=0
     externalLinearDragV=<0 0.5 0>
    }}
   skydive_FloatingDown {
    componentLinearAccScale=<0 0.5 0>
    componentLinearVelScale=<0 0.5 0>
    simSpaceSettings {
     maxAngularAcceleration=1000
     maxAngularVelocity=3
     maxLinearAcceleration=300
     maxLinearVelocity=700
     velocityScaleZ=0
     externalLinearDragV=<0 0.5 0>
    }
   }
   skydive_GoingUp {simSpaceSettings {
     maxAngularAcceleration=1000
     maxAngularVelocity=3
     maxLinearAcceleration=300
     maxLinearVelocity=700
     velocityScaleZ=0
     externalLinearDragV=<0 0.5 0>
    }}
   skydive_Parachuting {
    componentLinearAccScale=<0 0.1 0.3>
    componentLinearVelScale=<0 0.1 0.3>
    simSpaceSettings {
     maxAngularAcceleration=1000
     maxAngularVelocity=3
     maxLinearAcceleration=300
     maxLinearVelocity=700
     velocityScaleZ=0
     externalLinearDragV=<0 0.5 0>
    }
   }
   skydive_Surfing {simSpaceSettings {
     maxAngularAcceleration=1000
     maxAngularVelocity=3
     maxLinearAcceleration=300
     maxLinearVelocity=700
     velocityScaleZ=0
     externalLinearDragV=<0 0.5 0>
    }}
   swimming {simSpaceSettings {
     maxAngularAcceleration=1000
     maxAngularVelocity=3
     maxLinearAcceleration=300
     maxLinearVelocity=700
     velocityScaleZ=0
     externalLinearDragV=<0 0.5 0>
    }}
   tactical_Sprint {simSpaceSettings {
     maxAngularAcceleration=1000
     maxAngularVelocity=3
     maxLinearAcceleration=300
     maxLinearVelocity=700
     velocityScaleZ=0
     externalLinearDragV=<0 0.5 0>
    }}
   ziplining {simSpaceSettings {
     maxAngularAcceleration=1000
     maxAngularVelocity=3
     maxLinearAcceleration=300
     maxLinearVelocity=700
     velocityScaleZ=0
     externalLinearDragV=<0 0.5 0>
    }}
  }
 }
 ScaleRange {type=Vector2D, default=<0.2 1>}
 ScaleRange_2 {type=Vector2D, default=<0.4 1.1>}
 shakeFrequency {type=number, default=0.17}
 shakeFrequency2 {type=number, default=0.07}
 XScale {type=Vector}
 XScale_2 {type=Vector}
]
