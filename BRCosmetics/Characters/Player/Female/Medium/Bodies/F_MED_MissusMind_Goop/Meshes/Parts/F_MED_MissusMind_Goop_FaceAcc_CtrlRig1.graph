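# Control Rig for the goop face accessory on the female MissusMind Goop character, animating seven bone chains (A–G) parented to the head.
# A sequence first applies fixed transforms to goop master bones, then blends in corrective transforms weighted by a smoothed spherical pose reader tracking head orientation against spine_05.
# It then runs per-chain harmonics from each chain's first bone, using only noisy wave motion with per-chain amplitude and offset; pendulum and reach are disabled.
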ControlRigBlueprint /BRCosmetics/Characters/Player/Female/Medium/Bodies/F_MED_MissusMind_Goop/Meshes/Parts/F_MED_MissusMind_Goop_FaceAcc_CtrlRig1
graphs [
 RigVMFunctionLibrary {links [], nodes []}
 RigVMModel {nodes [
   RigUnit_BeginExecution = &RigUnit_BeginExecution()
   RigVMFunction_Sequence = &RigVMFunction_Sequence(ExecuteContext=@RigUnit_BeginExecution.ExecutePin)
   &RigUnit_ModifyTransforms(
    ItemToModify="((Item=(Type=Bone,Name=\"dyn_faceAcc_goop_A_mstr\"),Transform=(Rotation=(X=0,Y=-0.2884,Z=0,W=-0.9575),Translation=(X=0,Y=0,Z=0),Scale3D=(X=1,Y=1,Z=1))),(Item=(Type=Bone,Name=\"dyn_faceAcc_goop_B_mstr\"),Transform=(Rotation=(X=0,Y=-0.3864,Z=0,W=-0.9223),Transla ...(truncated)"
    Weight=1
    WeightMinimum=0
    ExecutePin=@RigVMFunction_Sequence.A
   )
   > n1 = &RigUnit_SphericalPoseReader(
    ActiveRegionScaleFactors=(PositiveWidth=1,NegativeWidth=1,PositiveHeight=0.75,NegativeHeight=0.75)
    ActiveRegionSize=0.15
    Debug=(bDrawDebug=false,bDraw2D=False,bDrawLocalAxes=False,DebugScale=25,DebugSegments=20,DebugThickness=0.25)
    DriverAxis=(X=0,Y=1,Z=0)
    DriverItem=(Type=Bone,Name="head")
    FalloffRegionScaleFactors=(PositiveWidth=1,NegativeWidth=1,PositiveHeight=0.5,NegativeHeight=0.5)
    FalloffSize=0.8
    FlipHeightScaling=False
    FlipWidthScaling=False
    OptionalParentItem=(Type=Bone,Name="spine_05")
    RotationOffset=(X=90,Y=0,Z=-90)
   )
   RigVMFunction_AlphaInterp = &RigVMFunction_AlphaInterp(InterpSpeedDecreasing=40, InterpSpeedIncreasing=40, Scale=1, bClampResult=False, bInterpResult=True, Value=@n1.OutputParam)
   > &RigUnit_ModifyTransforms(
    ItemToModify="((Item=(Type=Bone,Name=\"dyn_faceAcc_goop_A_mstr\"),Transform=(Rotation=(X=0.0694,Y=0.0349,Z=0.5264,W=-0.8467),Translation=(X=0,Y=0,Z=0),Scale3D=(X=1,Y=1,Z=1))),(Item=(Type=Bone,Name=\"dyn_faceAcc_goop_B_mstr\"),Transform=(Rotation=(X=0.0555,Y=0.0848,Z=-0.2269 ...(truncated)"
    WeightMinimum=0
    Weight=@RigVMFunction_AlphaInterp.Result
   )
   /Script/ControlRig.RigUnit_ChainHarmonicsPerItem(
    ChainRoot=(Type=Bone,Name="dyn_faceAcc_goop_A_01")
    DrawWorldOffset=(Rotation=(X=0,Y=0,Z=0,W=1),Translation=(X=0,Y=0,Z=0),Scale3D=(X=1,Y=1,Z=1))
    Pendulum=(bEnabled=False,PendulumStiffness=2,PendulumGravity=(X=0,Y=0,Z=0),PendulumBlend=0.75,PendulumDrag=0.98,PendulumMinimum=0,PendulumMaximum=0.1,PendulumEase=Linear,UnwindAxis=(X=0,Y=1,Z=0),UnwindMinimum=0.2,UnwindMaximum=0.05)
    Reach=(bEnabled=False,ReachTarget=(X=0,Y=0,Z=0),ReachAxis=(X=1,Y=0,Z=0),ReachMinimum=0,ReachMaximum=0,ReachEase=Linear)
    Speed=(X=1.5,Y=1.5,Z=1.5)
    Wave=(bEnabled=True,WaveFrequency=(X=1,Y=0.6,Z=0.8),WaveAmplitude=(X=0.1,Y=0.1,Z=0.2),WaveOffset=(X=0,Y=1,Z=2),WaveNoise=(X=0.05,Y=0.05,Z=0.05),WaveMinimum=0,WaveMaximum=1,WaveEase=Linear)
    WaveCurve=(EditorCurveData=(Keys=((),(Time=1,Value=1)),DefaultValue=340282346638528859811704183484516925440,PreInfinityExtrap=RCCE_Constant,PostInfinityExtrap=RCCE_Constant),ExternalCurve=None)
    bDrawDebug=false
    ExecutePin=@RigVMFunction_Sequence.B
   ) {is=struct RigVMUnitNode}
   > &RigUnit_ChainHarmonicsPerItem(
    ChainRoot=(Type=Bone,Name="dyn_faceAcc_goop_B_01")
    DrawWorldOffset=(Rotation=(X=0,Y=0,Z=0,W=1),Translation=(X=0,Y=0,Z=0),Scale3D=(X=1,Y=1,Z=1))
    Pendulum=(bEnabled=False,PendulumStiffness=2,PendulumGravity=(X=0,Y=0,Z=0),PendulumBlend=0.75,PendulumDrag=0.98,PendulumMinimum=0,PendulumMaximum=0.1,PendulumEase=Linear,UnwindAxis=(X=0,Y=1,Z=0),UnwindMinimum=0.2,UnwindMaximum=0.05)
    Reach=(bEnabled=False,ReachTarget=(X=0,Y=0,Z=0),ReachAxis=(X=1,Y=0,Z=0),ReachMinimum=0,ReachMaximum=0,ReachEase=Linear)
    Speed=(X=1.5,Y=1.5,Z=1.5)
    Wave=(bEnabled=True,WaveFrequency=(X=1,Y=0.6,Z=0.8),WaveAmplitude=(X=0.4,Y=0.3,Z=0.4),WaveOffset=(X=0,Y=3,Z=3),WaveNoise=(X=0.05,Y=0.05,Z=0.05),WaveMinimum=0,WaveMaximum=1,WaveEase=Linear)
    WaveCurve=(EditorCurveData=(Keys=((),(Time=1,Value=1)),DefaultValue=340282346638528859811704183484516925440,PreInfinityExtrap=RCCE_Constant,PostInfinityExtrap=RCCE_Constant),ExternalCurve=None)
    bDrawDebug=false
   )
   > &RigUnit_ChainHarmonicsPerItem(
    ChainRoot=(Type=Bone,Name="dyn_faceAcc_goop_C_01")
    DrawWorldOffset=(Rotation=(X=0,Y=0,Z=0,W=1),Translation=(X=0,Y=0,Z=0),Scale3D=(X=1,Y=1,Z=1))
    Pendulum=(bEnabled=False,PendulumStiffness=2,PendulumGravity=(X=0,Y=0,Z=0),PendulumBlend=0.75,PendulumDrag=0.98,PendulumMinimum=0,PendulumMaximum=0.1,PendulumEase=Linear,UnwindAxis=(X=0,Y=1,Z=0),UnwindMinimum=0.2,UnwindMaximum=0.05)
    Reach=(bEnabled=False,ReachTarget=(X=0,Y=0,Z=0),ReachAxis=(X=1,Y=0,Z=0),ReachMinimum=0,ReachMaximum=0,ReachEase=Linear)
    Speed=(X=1.5,Y=1.5,Z=1.5)
    Wave=(bEnabled=True,WaveFrequency=(X=1,Y=0.6,Z=0.8),WaveAmplitude=(X=0.2,Y=0.5,Z=0.6),WaveOffset=(X=0,Y=2,Z=1),WaveNoise=(X=0.05,Y=0.05,Z=0.05),WaveMinimum=0,WaveMaximum=1,WaveEase=Linear)
    WaveCurve=(EditorCurveData=(Keys=((),(Time=1,Value=1)),DefaultValue=340282346638528859811704183484516925440,PreInfinityExtrap=RCCE_Constant,PostInfinityExtrap=RCCE_Constant),ExternalCurve=None)
    bDrawDebug=false
   )
   > &RigUnit_ChainHarmonicsPerItem(
    ChainRoot=(Type=Bone,Name="dyn_faceAcc_goop_D_01")
    DrawWorldOffset=(Rotation=(X=0,Y=0,Z=0,W=1),Translation=(X=0,Y=0,Z=0),Scale3D=(X=1,Y=1,Z=1))
    Pendulum=(bEnabled=False,PendulumStiffness=2,PendulumGravity=(X=0,Y=0,Z=0),PendulumBlend=0.75,PendulumDrag=0.98,PendulumMinimum=0,PendulumMaximum=0.1,PendulumEase=Linear,UnwindAxis=(X=0,Y=1,Z=0),UnwindMinimum=0.2,UnwindMaximum=0.05)
    Reach=(bEnabled=False,ReachTarget=(X=0,Y=0,Z=0),ReachAxis=(X=1,Y=0,Z=0),ReachMinimum=0,ReachMaximum=0,ReachEase=Linear)
    Speed=(X=1.5,Y=1.5,Z=1.5)
    Wave=(bEnabled=True,WaveFrequency=(X=1,Y=0.6,Z=0.8),WaveAmplitude=(X=0.2,Y=0.1,Z=0.3),WaveOffset=(X=0,Y=0,Z=1),WaveNoise=(X=0.05,Y=0.05,Z=0.05),WaveMinimum=0,WaveMaximum=1,WaveEase=Linear)
    WaveCurve=(EditorCurveData=(Keys=((),(Time=1,Value=1)),DefaultValue=340282346638528859811704183484516925440,PreInfinityExtrap=RCCE_Constant,PostInfinityExtrap=RCCE_Constant),ExternalCurve=None)
    bDrawDebug=false
   )
   > &RigUnit_ChainHarmonicsPerItem(
    ChainRoot=(Type=Bone,Name="dyn_faceAcc_goop_E_01")
    DrawWorldOffset=(Rotation=(X=0,Y=0,Z=0,W=1),Translation=(X=0,Y=0,Z=0),Scale3D=(X=1,Y=1,Z=1))
    Pendulum=(bEnabled=False,PendulumStiffness=2,PendulumGravity=(X=0,Y=0,Z=0),PendulumBlend=0.75,PendulumDrag=0.98,PendulumMinimum=0,PendulumMaximum=0.1,PendulumEase=Linear,UnwindAxis=(X=0,Y=1,Z=0),UnwindMinimum=0.2,UnwindMaximum=0.05)
    Reach=(bEnabled=False,ReachTarget=(X=0,Y=0,Z=0),ReachAxis=(X=1,Y=0,Z=0),ReachMinimum=0,ReachMaximum=0,ReachEase=Linear)
    Speed=(X=1.5,Y=1.5,Z=1.5)
    Wave=(bEnabled=True,WaveFrequency=(X=1,Y=0.6,Z=0.8),WaveAmplitude=(X=0.1,Y=0.1,Z=0.3),WaveOffset=(X=0,Y=2,Z=2),WaveNoise=(X=0.05,Y=0.05,Z=0.05),WaveMinimum=0,WaveMaximum=1,WaveEase=Linear)
    WaveCurve=(EditorCurveData=(Keys=((),(Time=1,Value=1)),DefaultValue=340282346638528859811704183484516925440,PreInfinityExtrap=RCCE_Constant,PostInfinityExtrap=RCCE_Constant),ExternalCurve=None)
    bDrawDebug=false
   )
   > &RigUnit_ChainHarmonicsPerItem(
    ChainRoot=(Type=Bone,Name="dyn_faceAcc_goop_F_01")
    DrawWorldOffset=(Rotation=(X=0,Y=0,Z=0,W=1),Translation=(X=0,Y=0,Z=0),Scale3D=(X=1,Y=1,Z=1))
    Pendulum=(bEnabled=False,PendulumStiffness=2,PendulumGravity=(X=0,Y=0,Z=0),PendulumBlend=0.75,PendulumDrag=0.98,PendulumMinimum=0,PendulumMaximum=0.1,PendulumEase=Linear,UnwindAxis=(X=0,Y=1,Z=0),UnwindMinimum=0.2,UnwindMaximum=0.05)
    Reach=(bEnabled=False,ReachTarget=(X=0,Y=0,Z=0),ReachAxis=(X=1,Y=0,Z=0),ReachMinimum=0,ReachMaximum=0,ReachEase=Linear)
    Speed=(X=1.5,Y=1.5,Z=1.5)
    Wave=(bEnabled=True,WaveFrequency=(X=1,Y=0.6,Z=0.8),WaveAmplitude=(X=0.1,Y=0.1,Z=0.1),WaveOffset=(X=0,Y=1,Z=1),WaveNoise=(X=0.05,Y=0.05,Z=0.05),WaveMinimum=0,WaveMaximum=1,WaveEase=Linear)
    WaveCurve=(EditorCurveData=(Keys=((),(Time=1,Value=1)),DefaultValue=340282346638528859811704183484516925440,PreInfinityExtrap=RCCE_Constant,PostInfinityExtrap=RCCE_Constant),ExternalCurve=None)
    bDrawDebug=false
   )
   > &RigUnit_ChainHarmonicsPerItem(
    ChainRoot=(Type=Bone,Name="dyn_faceAcc_goop_G_01")
    DrawWorldOffset=(Rotation=(X=0,Y=0,Z=0,W=1),Translation=(X=0,Y=0,Z=0),Scale3D=(X=1,Y=1,Z=1))
    Pendulum=(bEnabled=False,PendulumStiffness=2,PendulumGravity=(X=0,Y=0,Z=0),PendulumBlend=0.75,PendulumDrag=0.98,PendulumMinimum=0,PendulumMaximum=0.1,PendulumEase=Linear,UnwindAxis=(X=0,Y=1,Z=0),UnwindMinimum=0.2,UnwindMaximum=0.05)
    Reach=(bEnabled=False,ReachTarget=(X=0,Y=0,Z=0),ReachAxis=(X=1,Y=0,Z=0),ReachMinimum=0,ReachMaximum=0,ReachEase=Linear)
    Speed=(X=1.5,Y=1.5,Z=1.5)
    Wave=(bEnabled=True,WaveFrequency=(X=1,Y=0.6,Z=0.8),WaveAmplitude=(X=0.2,Y=0.3,Z=0.3),WaveOffset=(X=0,Y=4,Z=4),WaveNoise=(X=0.05,Y=0.05,Z=0.05),WaveMinimum=0,WaveMaximum=1,WaveEase=Linear)
    WaveCurve=(EditorCurveData=(Keys=((),(Time=1,Value=1)),DefaultValue=340282346638528859811704183484516925440,PreInfinityExtrap=RCCE_Constant,PostInfinityExtrap=RCCE_Constant),ExternalCurve=None)
    bDrawDebug=false
   )
  ]}
]
hierarchy [
 root {loc=<0 0 0>, rot=<0 0 0 1>}
 pelvis {parent=root, loc=<0 -0.5301 98.0571>, rot=<0.0127 0.707 -0.0127 -0.707>}
 spine_01 {parent=pelvis, loc=<6.1208 -0.9328 0>, rot=<0 0 0.034 -0.9994>}
 spine_02 {parent=spine_01, loc=<6.5996 -0.2095 0>, rot=<0 0 0.0238 -0.9997>}
 spine_03 {parent=spine_02, loc=<6.4914 -0.1237 0>, rot=<0 0 0.0149 -0.9999>}
 spine_04 {parent=spine_03, loc=<6.9598 0.0794 0>, rot=<0 0 -0.0132 -0.9999>}
 spine_05 {parent=spine_04, loc=<6.9987 0 0>, rot=<0 0 -0.0015 -1>}
 neck_01 {parent=spine_05, loc=<11.3138 0.1529 0>, rot=<0 0 -0.2003 -0.9797>}
 neck_02 {parent=neck_01, loc=<4.5627 0 0>, rot=<0 0 0.0504 -0.9987>}
 head {parent=neck_02, loc=<5.3914 0 0>, rot=<0 0 0.115 -0.9934>}
 dyn_faceAcc_goop_A_mstr {parent=head, loc=<2.1123 -4.9119 -5.654>, rot=<0.1236 -0.9682 -0.1617 0.1458>}
 dyn_faceAcc_goop_A_sim {parent=dyn_faceAcc_goop_A_mstr, loc=<0 0 0>, rot=<0 0 0 1>}
 dyn_faceAcc_goop_A_01 {parent=dyn_faceAcc_goop_A_sim, loc=<0 0 0>, rot=<0 0 0 1>}
 dyn_faceAcc_goop_A_02 {parent=dyn_faceAcc_goop_A_01, loc=<5.0369 0 0>, rot=<-0.0552 0.0782 0.0979 -0.9906>}
 dyn_faceAcc_goop_A_03 {parent=dyn_faceAcc_goop_A_02, loc=<4.978 0 0>, rot=<-0.0363 -0.0185 0.3156 -0.948>}
 dyn_faceAcc_goop_A_04 {parent=dyn_faceAcc_goop_A_03, loc=<4.745 0 0>, rot=<-0.0315 -0.0163 0.5249 -0.8504>}
 dyn_faceAcc_goop_A_05 {parent=dyn_faceAcc_goop_A_04, loc=<4.843 0 0>, rot=<-0.0345 -0.0079 0.2456 -0.9687>}
 dyn_faceAcc_goop_B_mstr {parent=head, loc=<3.344 -6.8119 -2.4227>, rot=<0.0631 -0.9943 -0.0516 0.0681>}
 dyn_faceAcc_goop_B_sim {parent=dyn_faceAcc_goop_B_mstr, loc=<0 0 0>, rot=<0 0 0 1>}
 dyn_faceAcc_goop_B_01 {parent=dyn_faceAcc_goop_B_sim, loc=<0 0 0>, rot=<0 0 0 1>}
 dyn_faceAcc_goop_B_02 {parent=dyn_faceAcc_goop_B_01, loc=<5.32 0 0>, rot=<0.0131 0.06 0.2228 -0.9729>}
 dyn_faceAcc_goop_B_03 {parent=dyn_faceAcc_goop_B_02, loc=<4.7808 0 0>, rot=<-0.0298 -0.0182 0.5201 -0.8534>}
 dyn_faceAcc_goop_B_04 {parent=dyn_faceAcc_goop_B_03, loc=<4.6834 0 0>, rot=<0.0071 -0.0135 0.3894 -0.921>}
 dyn_faceAcc_goop_B_05 {parent=dyn_faceAcc_goop_B_04, loc=<4.8685 0 0>, rot=<0.005 -0.0028 0.1792 -0.9838>}
 dyn_faceAcc_goop_C_mstr {parent=head, loc=<6.0026 -7.7676 -0.5013>, rot=<0.2811 -0.951 0.121 0.0445>}
 dyn_faceAcc_goop_C_sim {parent=dyn_faceAcc_goop_C_mstr, loc=<0 0 0>, rot=<0 0 0 1>}
 dyn_faceAcc_goop_C_01 {parent=dyn_faceAcc_goop_C_sim, loc=<0 0 0>, rot=<0 0 0 1>}
 dyn_faceAcc_goop_C_02 {parent=dyn_faceAcc_goop_C_01, loc=<4.0935 0 0>, rot=<-0.0272 -0.0335 0.5121 -0.8578>}
 dyn_faceAcc_goop_C_03 {parent=dyn_faceAcc_goop_C_02, loc=<3.6273 0 0>, rot=<0.0041 -0.0171 0.1582 -0.9873>}
 dyn_faceAcc_goop_D_mstr {parent=head, loc=<5.4552 -7.3205 2.1562>, rot=<0.0529 -0.9805 0.1287 -0.1385>}
 dyn_faceAcc_goop_D_sim {parent=dyn_faceAcc_goop_D_mstr, loc=<0 0 0>, rot=<0 0 0 1>}
 dyn_faceAcc_goop_D_01 {parent=dyn_faceAcc_goop_D_sim, loc=<0 0 0>, rot=<0 0 0 1>}
 dyn_faceAcc_goop_D_02 {parent=dyn_faceAcc_goop_D_01, loc=<3.7142 0 0>, rot=<0.0523 -0.0535 0.0744 -0.9944>}
 dyn_faceAcc_goop_D_03 {parent=dyn_faceAcc_goop_D_02, loc=<3.5339 0 0>, rot=<-0.0366 -0.1191 0.3167 -0.9403>}
 dyn_faceAcc_goop_D_04 {parent=dyn_faceAcc_goop_D_03, loc=<3.5469 0 0>, rot=<-0.009 -0.0262 0.3646 -0.9308>}
 dyn_faceAcc_goop_D_05 {parent=dyn_faceAcc_goop_D_04, loc=<3.6087 0 0>, rot=<-0.0016 0.0084 0.146 -0.9893>}
 dyn_faceAcc_goop_E_mstr {parent=head, loc=<3.4787 -6.0253 3.9303>, rot=<0.0593 -0.9584 0.1948 -0.1998>}
 dyn_faceAcc_goop_E_sim {parent=dyn_faceAcc_goop_E_mstr, loc=<0 0 0>, rot=<0 0 0 1>}
 dyn_faceAcc_goop_E_01 {parent=dyn_faceAcc_goop_E_sim, loc=<0 0 0>, rot=<0 0 0 1>}
 dyn_faceAcc_goop_E_02 {parent=dyn_faceAcc_goop_E_01, loc=<3.5931 0.2656 -0.0088>, rot=<0.0188 -0.0208 -0.0438 -0.9986>}
 dyn_faceAcc_goop_E_03 {parent=dyn_faceAcc_goop_E_02, loc=<3.6863 0 0>, rot=<0.0258 -0.0133 0.1416 -0.9895>}
 dyn_faceAcc_goop_E_04 {parent=dyn_faceAcc_goop_E_03, loc=<3.6924 0 0>, rot=<0.1705 0.0145 0.3498 -0.9211>}
 dyn_faceAcc_goop_E_05 {parent=dyn_faceAcc_goop_E_04, loc=<3.7723 0 0>, rot=<0.0692 -0.0431 0.3136 -0.9461>}
 dyn_faceAcc_goop_E_06 {parent=dyn_faceAcc_goop_E_05, loc=<3.7473 0 0>, rot=<0.0419 -0.0969 0.2776 -0.9549>}
 dyn_faceAcc_goop_F_mstr {parent=head, loc=<0.5934 -4.1783 5.4805>, rot=<0.1163 -0.9285 0.266 -0.2313>}
 dyn_faceAcc_goop_F_sim {parent=dyn_faceAcc_goop_F_mstr, loc=<0 0 0>, rot=<0 0 0 1>}
 dyn_faceAcc_goop_F_01 {parent=dyn_faceAcc_goop_F_sim, loc=<0 0 0>, rot=<0 0 0 1>}
 dyn_faceAcc_goop_F_02 {parent=dyn_faceAcc_goop_F_01, loc=<3.584 0 0>, rot=<0.1194 -0.0544 0.0935 -0.9869>}
 dyn_faceAcc_goop_F_03 {parent=dyn_faceAcc_goop_F_02, loc=<3.5694 0 0>, rot=<-0.0468 -0.0653 0.2509 -0.9647>}
 dyn_faceAcc_goop_F_04 {parent=dyn_faceAcc_goop_F_03, loc=<3.6163 0 0>, rot=<0.0668 -0.0393 0.28 -0.9568>}
 dyn_faceAcc_goop_G_mstr {parent=head, loc=<5.8602 -6.1891 0.3537>, rot=<-0.0067 -0.9924 -0.0772 -0.0958>}
 dyn_faceAcc_goop_G_sim {parent=dyn_faceAcc_goop_G_mstr, loc=<0 0 0>, rot=<0 0 0 1>}
 dyn_faceAcc_goop_G_01 {parent=dyn_faceAcc_goop_G_sim, loc=<0 0 0>, rot=<0 0 0 1>}
 dyn_faceAcc_goop_G_02 {parent=dyn_faceAcc_goop_G_01, loc=<3.7453 0 0>, rot=<0.011 -0.0341 -0.0415 -0.9985>}
 dyn_faceAcc_goop_G_03 {parent=dyn_faceAcc_goop_G_02, loc=<3.5752 0 0>, rot=<-0.0106 -0.058 0.1378 -0.9887>}
]
