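# Physics asset for the face accessory of the female MissusMind Goop character cosmetic, providing physics bodies for seven dangling goop pieces (A to G) around the head.
# Each piece is a two-bone chain: a small master capsule constrained to the head capsule with free swing and twist, and a longer offset sim capsule constrained to that master.
# The sim joints have limited swing and very tight twist, so each piece can sway but barely spin.
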PhysicsAsset /BRCosmetics/Characters/Player/Female/Medium/Bodies/F_MED_MissusMind_Goop/Meshes/Parts/F_MED_MissusMind_Goop_FaceAcc_Physics
bodies {
 dyn_faceAcc_goop_A_mstr [{shapeType=Capsule, length=0.505, radius=0.505, shapeName=dyn_faceAcc_goop_A_mstr_capsule, rotation=<rot -90 0 0>}]
 dyn_faceAcc_goop_A_sim [{shapeType=Capsule, length=7.0333, radius=1.1699, shapeName=dyn_faceAcc_goop_A_sim_capsule, center=<4.4367 0 0>, rotation=<rot -90 180 180>}]
 dyn_faceAcc_goop_B_mstr [{shapeType=Capsule, length=0.505, radius=0.505, shapeName=dyn_faceAcc_goop_B_mstr_capsule, rotation=<rot -90 0 0>}]
 dyn_faceAcc_goop_B_sim [{shapeType=Capsule, length=7.0333, radius=1.1699, shapeName=dyn_faceAcc_goop_B_sim_capsule, center=<4.4367 0 0>, rotation=<rot -90 180 180>}]
 dyn_faceAcc_goop_C_mstr [{shapeType=Capsule, length=0.505, radius=0.505, shapeName=dyn_faceAcc_goop_C_mstr_capsule, rotation=<rot -90 0 0>}]
 dyn_faceAcc_goop_C_sim [{shapeType=Capsule, length=7.0333, radius=1.1699, shapeName=dyn_faceAcc_goop_C_sim_capsule, center=<4.4367 0 0>, rotation=<rot -90 180 180>}]
 dyn_faceAcc_goop_D_mstr [{shapeType=Capsule, length=0.505, radius=0.505, shapeName=dyn_faceAcc_goop_D_mstr_capsule, rotation=<rot -90 0 0>}]
 dyn_faceAcc_goop_D_sim [{shapeType=Capsule, length=7.0333, radius=1.1699, shapeName=dyn_faceAcc_goop_D_sim_capsule, center=<4.4367 0 0>, rotation=<rot -90 180 180>}]
 dyn_faceAcc_goop_E_mstr [{shapeType=Capsule, length=0.505, radius=0.505, shapeName=dyn_faceAcc_goop_E_mstr_capsule, rotation=<rot -90 0 0>}]
 dyn_faceAcc_goop_E_sim [{shapeType=Capsule, length=7.0333, radius=1.1699, shapeName=dyn_faceAcc_goop_E_sim_capsule, center=<4.4367 0 0>, rotation=<rot -90 180 180>}]
 dyn_faceAcc_goop_F_mstr [{shapeType=Capsule, length=0.505, radius=0.505, shapeName=dyn_faceAcc_goop_F_mstr_capsule, rotation=<rot -90 0 0>}]
 dyn_faceAcc_goop_F_sim [{shapeType=Capsule, length=7.0333, radius=1.1699, shapeName=dyn_faceAcc_goop_F_sim_capsule, center=<4.4367 0 0>, rotation=<rot -90 180 180>}]
 dyn_faceAcc_goop_G_mstr [{shapeType=Capsule, length=0.505, radius=0.505, shapeName=dyn_faceAcc_goop_G_mstr_capsule, rotation=<rot -90 0 0>}]
 dyn_faceAcc_goop_G_sim [{shapeType=Capsule, length=7.0333, radius=1.1699, shapeName=dyn_faceAcc_goop_G_sim_capsule, center=<4.4367 0 0>, rotation=<rot -90 180 180>}]
 head [{shapeType=Capsule, length=9.4508, radius=2.2986, shapeName=head_capsule, rotation=<rot -90 0 0>}]
}
constraints [
 {bone1Name=dyn_faceAcc_goop_A_sim, bone2Name=dyn_faceAcc_goop_A_mstr, swing1LimitDegrees=35, swing2LimitDegrees=45, twistLimitDegrees=2}
 {
  bone1Name=dyn_faceAcc_goop_A_mstr
  bone2Name=head
  swing1LimitDegrees=45
  swing1Motion=Free
  swing2LimitDegrees=45
  swing2Motion=Free
  twistLimitDegrees=45
  twistMotion=Free
 }
 {
  bone1Name=dyn_faceAcc_goop_B_mstr
  bone2Name=head
  swing1LimitDegrees=45
  swing1Motion=Free
  swing2LimitDegrees=45
  swing2Motion=Free
  twistLimitDegrees=45
  twistMotion=Free
 }
 {bone1Name=dyn_faceAcc_goop_B_sim, bone2Name=dyn_faceAcc_goop_B_mstr, swing1LimitDegrees=35, swing2LimitDegrees=45, twistLimitDegrees=2}
 {
  bone1Name=dyn_faceAcc_goop_C_mstr
  bone2Name=head
  swing1LimitDegrees=45
  swing1Motion=Free
  swing2LimitDegrees=45
  swing2Motion=Free
  twistLimitDegrees=45
  twistMotion=Free
 }
 {bone1Name=dyn_faceAcc_goop_C_sim, bone2Name=dyn_faceAcc_goop_C_mstr, swing1LimitDegrees=35, swing2LimitDegrees=45, twistLimitDegrees=2}
 {
  bone1Name=dyn_faceAcc_goop_D_mstr
  bone2Name=head
  swing1LimitDegrees=45
  swing1Motion=Free
  swing2LimitDegrees=45
  swing2Motion=Free
  twistLimitDegrees=45
  twistMotion=Free
 }
 {bone1Name=dyn_faceAcc_goop_D_sim, bone2Name=dyn_faceAcc_goop_D_mstr, swing1LimitDegrees=35, swing2LimitDegrees=45, twistLimitDegrees=2}
 {
  bone1Name=dyn_faceAcc_goop_E_mstr
  bone2Name=head
  swing1LimitDegrees=45
  swing1Motion=Free
  swing2LimitDegrees=45
  swing2Motion=Free
  twistLimitDegrees=45
  twistMotion=Free
 }
 {bone1Name=dyn_faceAcc_goop_E_sim, bone2Name=dyn_faceAcc_goop_E_mstr, swing1LimitDegrees=35, swing2LimitDegrees=45, twistLimitDegrees=2}
 {
  bone1Name=dyn_faceAcc_goop_F_mstr
  bone2Name=head
  swing1LimitDegrees=45
  swing1Motion=Free
  swing2LimitDegrees=45
  swing2Motion=Free
  twistLimitDegrees=45
  twistMotion=Free
 }
 {bone1Name=dyn_faceAcc_goop_F_sim, bone2Name=dyn_faceAcc_goop_F_mstr, swing1LimitDegrees=35, swing2LimitDegrees=45, twistLimitDegrees=2}
 {
  bone1Name=dyn_faceAcc_goop_G_mstr
  bone2Name=head
  swing1LimitDegrees=45
  swing1Motion=Free
  swing2LimitDegrees=45
  swing2Motion=Free
  twistLimitDegrees=45
  twistMotion=Free
 }
 {bone1Name=dyn_faceAcc_goop_G_sim, bone2Name=dyn_faceAcc_goop_G_mstr, swing1LimitDegrees=35, swing2LimitDegrees=45, twistLimitDegrees=2}
]
